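# Physics asset for the face accessory part of the male medium TwiceBaked Hood body cosmetic, covering a left-side hood bone chain and the head.
# Capsule bodies sit on the head, a tiny master hood bone and the hood bone below it in the chain.
# The master bone's constraint to the head leaves swing and twist free, while the child bone's constraint to the master has narrow swing limits and a zero twist limit.
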PhysicsAsset /BRCosmetics/Characters/Player/Male/Medium/Bodies/M_MED_TwiceBaked_Hood/Meshes/Parts/M_MED_TwiceBaked_Hood_FaceAcc_Physics1
bodies {
 dyn_hoodOn_l_01 [{shapeType=Capsule, length=4.9044, radius=3.3061, shapeName=dyn_hoodOn_l_01_capsule, center=<0 0 -3.3029>}]
 dyn_hoodOn_l_MSTR [{shapeType=Capsule, length=0.505, radius=0.505, shapeName=dyn_hoodOn_l_MSTR_capsule}]
 head [{shapeType=Capsule, length=19.3308, radius=9.7391, shapeName=head_capsule, rotation=<rot 90 0 0>}]
}
constraints [
 {
  bone1Name=dyn_hoodOn_l_MSTR
  bone2Name=head
  swing1LimitDegrees=45
  swing1Motion=Free
  swing2LimitDegrees=45
  swing2Motion=Free
  twistLimitDegrees=45
  twistMotion=Free
 }
 {bone1Name=dyn_hoodOn_l_01, bone2Name=dyn_hoodOn_l_MSTR, swing1LimitDegrees=20, swing2LimitDegrees=20, twistLimitDegrees=0}
]
